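# Physics asset for the hair of the male medium Citadel Armor character's face accessory part, apparently the frontend variant by its name.
# Capsule chains cover back, side, left-front and top hair, each rooted in a tiny MSTR body with free constraints to the head.
# Chain links have tight swing limits and zero twist limits, while the top piece is fully locked.
# Head, neck, upper spine and clavicles carry larger body collision shapes.
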PhysicsAsset /BRCosmetics/Characters/Player/Male/Medium/Bodies/M_MED_Citadel_Armor_Hair/Meshes/Parts/M_MED_Citadel_Armor_FaceAcc_PhysicsFrontend
bodies {
 clavicle_l [{
   shapeType=Capsule
   length=14.2167
   radius=10.5781
   shapeName=clavicle_l_capsule
   center=<6.3334 -1.3389 -1.4203>
   rotation=<rot 70 -171.2417 -179.9999>
  }]
 clavicle_r [{
   shapeType=Capsule
   length=12.9667
   radius=11.8281
   shapeName=clavicle_l_capsule
   center=<-6.1168 -0.067 1.4203>
   rotation=<rot 70 -171.2416 0.0002>
  }]
 dyn_hairBk_c_01 [{shapeType=Capsule, length=0.1, radius=0.1, shapeName=dyn_hairBk_c_01_capsule, center=<1.7999 -2 0>}]
 dyn_hairBk_c_02 [{shapeType=Capsule, length=0.1, radius=0.1, shapeName=dyn_hairBk_c_02_capsule, center=<1.7999 -2 0>}]
 dyn_hairBk_c_03 [{shapeType=Capsule, length=1.4674, radius=3.2724, shapeName=dyn_hairBk_c_03_capsule, center=<1.7999 -1.7 0>}]
 dyn_hairBk_c_04 [{shapeType=Capsule, length=1.4674, radius=3.2724, shapeName=dyn_hairBk_c_04_capsule, center=<1.5561 -1.75 0>}]
 dyn_hairBk_c_05 [{shapeType=Capsule, length=1.2174, radius=3.5224, shapeName=dyn_hairBk_c_05_capsule, center=<1.4935 -1.8 0>}]
 dyn_hairBk_c_06 [{shapeType=Capsule, length=3.8479, radius=3.1419, shapeName=dyn_hairBk_c_06_capsule, center=<1.6385 -0.7 -0.5>, rotation=<rot 90 0 0>}]
 dyn_hairBk_c_MSTR [{shapeType=Capsule, length=0.1, radius=0.1, shapeName=dyn_hairBk_c_MSTR_capsule, rotation=<rot 90 0 0>}]
 dyn_hairBk_l_01 [{shapeType=Capsule, length=0.1, radius=0.1, shapeName=dyn_hairBk_l_01_capsule, center=<1.7999 -1.5 0>, rotation=<rot 0 0 -13.9608>}]
 dyn_hairBk_l_02 [{shapeType=Capsule, length=0.1, radius=0.1, shapeName=dyn_hairBk_l_02_capsule, center=<1.7999 0 0>, rotation=<rot 0 0 -22.8548>}]
 dyn_hairBk_l_03 [{shapeType=Capsule, length=2.2031, radius=2.9958, shapeName=dyn_hairBk_l_03_capsule, center=<1.7999 0 0>, rotation=<rot 0 0 -14.77>}]
 dyn_hairBk_l_04 [{shapeType=Capsule, length=2.7488, radius=2.4501, shapeName=dyn_hairBk_l_04_capsule, center=<1.7999 -0.7 -0.2>, rotation=<rot 0 0 -18.5861>}]
 dyn_hairBk_l_05 [{shapeType=Capsule, length=3.01, radius=2.864, shapeName=dyn_hairBk_l_05_capsule, center=<2.3321 -1.7 -0.8>, rotation=<rot 67.6842 0 -6.5975>}]
 dyn_hairBk_l_MSTR [{shapeType=Capsule, length=0.1, radius=0.1, shapeName=dyn_hairBk_l_MSTR_capsule, rotation=<rot 90 0 0>}]
 dyn_hairBk_r_01 [{shapeType=Capsule, length=0.1, radius=0.1, shapeName=dyn_hairBk_r_01_capsule, center=<1.7999 -1.5 0>}]
 dyn_hairBk_r_02 [{shapeType=Capsule, length=0.1, radius=0.1, shapeName=dyn_hairBk_r_02_capsule, center=<1.7999 -1.5 0>}]
 dyn_hairBk_r_03 [{shapeType=Capsule, length=1.8188, radius=3.5193, shapeName=dyn_hairBk_r_03_capsule, center=<1.7999 -1.5 0>, rotation=<rot 0 0 -21.9375>}]
 dyn_hairBk_r_04 [{shapeType=Capsule, length=3, radius=2.584, shapeName=dyn_hairBk_r_04_capsule, center=<1.7999 -1.5 0>, rotation=<rot -0.0089 0.0213 -6.0193>}]
 dyn_hairBk_r_05 [{
   shapeType=Capsule
   length=4.1709
   radius=3.6005
   shapeName=dyn_hairBk_r_05_capsule
   center=<1.7999 -0.6 0>
   rotation=<rot 90 40.5772 16.8446>
  }]
 dyn_hairBk_r_MSTR [{shapeType=Capsule, length=0.1, radius=0.1, shapeName=dyn_hairBk_r_MSTR_capsule, rotation=<rot 90 0 0>}]
 dyn_hairMeshFront_l_01 [{shapeType=Capsule, length=1.8937, radius=0.6682, shapeName=dyn_hairMeshFront_l_01_capsule, center=<1.1843 0 0>, rotation=<rot 90 0 0>}]
 dyn_hairMeshFront_l_02 [{shapeType=Capsule, length=3.9824, radius=0.6682, shapeName=dyn_hairMeshFront_l_02_capsule, center=<1.7999 0 0>, rotation=<rot 90 0 0>}]
 dyn_hairMeshFront_l_03 [{shapeType=Capsule, length=4.2829, radius=0.6682, shapeName=dyn_hairMeshFront_l_03_capsule, center=<1.7999 0 0>, rotation=<rot 90 0 0>}]
 dyn_hairMeshFront_l_04 [{shapeType=Capsule, length=4.2527, radius=0.6682, shapeName=dyn_hairMeshFront_l_04_capsule, center=<1.7999 0 0>, rotation=<rot 90 0 0>}]
 dyn_hairMeshFront_l_05 [{shapeType=Capsule, length=3.5197, radius=0.6682, shapeName=dyn_hairMeshFront_l_05_capsule, center=<1.7999 0 0>, rotation=<rot 90 0 0>}]
 dyn_hairMeshFront_l_06 [{shapeType=Capsule, length=3.2979, radius=0.6682, shapeName=dyn_hairMeshFront_l_06_capsule, center=<1.8751 0 0>, rotation=<rot 90 0 0>}]
 dyn_hairMeshFront_l_MSTR [{shapeType=Capsule, length=0.1, radius=0.1, shapeName=dyn_hairMeshFront_l_MSTR_capsule, rotation=<rot 90 0 0>}]
 dyn_hairMesh_l_01 [{
   shapeType=Capsule
   length=3.8355
   radius=0.9988
   shapeName=dyn_hairMesh_l_01_capsule
   center=<1.9518 0.9122 -0.3>
   rotation=<rot 74.8047 0 0.0001>
  }]
 dyn_hairMesh_l_02 [{shapeType=Capsule, length=3.9093, radius=1.3407, shapeName=dyn_hairMesh_l_02_capsule, center=<1.7999 0.3809 0>, rotation=<rot 90 0 0>}]
 dyn_hairMesh_l_03 [{shapeType=Capsule, length=3.6398, radius=1.3407, shapeName=dyn_hairMesh_l_03_capsule, center=<3.024 0 0>, rotation=<rot 90 0 0>}]
 dyn_hairMesh_l_MSTR [{shapeType=Capsule, length=0.1, radius=0.1, shapeName=dyn_hairMesh_l_MSTR_capsule, rotation=<rot 90 0 0>}]
 dyn_hairMesh_r_01 [{shapeType=Capsule, length=3.6357, radius=0.9524, shapeName=dyn_hairMesh_r_01_capsule, center=<1.9479 -0.15 0>, rotation=<rot 90 0 0>}]
 dyn_hairMesh_r_02 [{shapeType=Capsule, length=7.3984, radius=0.8407, shapeName=dyn_hairMesh_r_02_capsule, center=<2.6062 0 0>, rotation=<rot 80 180 180>}]
 dyn_hairMesh_r_03 [{shapeType=Capsule, length=8.6522, radius=1.08, shapeName=dyn_hairMesh_r_03_capsule, center=<2.5521 0 0>, rotation=<rot 80 0 0>}]
 dyn_hairMesh_r_MSTR [{shapeType=Capsule, length=0.1, radius=0.1, shapeName=dyn_hairMesh_r_MSTR_capsule, rotation=<rot 90 0 0>}]
 dyn_hairTop_c_01 [{shapeType=Capsule, length=1.9118, radius=2, shapeName=dyn_hairTop_c_01_capsule, center=<-0.5074 0 0>}]
 dyn_hairTop_c_MSTR [{shapeType=Capsule, length=0.1, radius=0.1, shapeName=dyn_hairTop_c_MSTR_capsule, rotation=<rot 90 0 0>}]
 head [
  {shapeType=Capsule, length=4.9342, radius=6.4153, center=<4.9817 -0.2702 0>, rotation=<rot 90 0 0>}
  {shapeType=Capsule, length=6.4866, radius=6.0485, shapeName=head_capsule, center=<4.0735 3.4832 0>, rotation=<rot 90 -178.2101 -171.0779>}
 ]
 neck_02 [
  {shapeType=Capsule, length=7.6051, radius=5.1644, shapeName=neck_02_capsule, center=<1.8504 -1.0489 0>, rotation=<rot 90 0 0>}
  {shapeType=Capsule, length=30.3814, radius=1.4963, shapeName=neck_02_capsule}
 ]
 spine_05 [
  {shapeType=Sphere, radius=10.8031, shapeName=spine_05_sphere, center=<4.8672 -0.7071 0>}
  {shapeType=Sphere, radius=14.5592, shapeName=spine_05_sphere, center=<4.0124 -0.1707 0>}
  {shapeType=Capsule, length=12.7624, radius=14.3898, shapeName=spine_05_capsule, center=<-1.3069 -2.1572 0>, rotation=<rot 0 -90 0>}
 ]
}
constraints [
 {
  bone1Name=dyn_hairBk_c_MSTR
  bone2Name=head
  swing1LimitDegrees=45
  swing1Motion=Free
  swing2LimitDegrees=45
  swing2Motion=Free
  twistLimitDegrees=45
  twistMotion=Free
 }
 {bone1Name=dyn_hairBk_c_01, bone2Name=dyn_hairBk_c_MSTR, swing1LimitDegrees=20, swing2LimitDegrees=20, twistLimitDegrees=0}
 {bone1Name=dyn_hairBk_c_02, bone2Name=dyn_hairBk_c_01, swing1LimitDegrees=20, swing2LimitDegrees=20, twistLimitDegrees=0}
 {bone1Name=dyn_hairBk_c_03, bone2Name=dyn_hairBk_c_02, swing1LimitDegrees=20, swing2LimitDegrees=10, twistLimitDegrees=0}
 {bone1Name=dyn_hairBk_c_04, bone2Name=dyn_hairBk_c_03, swing1LimitDegrees=20, swing2LimitDegrees=10, twistLimitDegrees=0}
 {bone1Name=dyn_hairBk_c_05, bone2Name=dyn_hairBk_c_04, swing1LimitDegrees=20, swing2LimitDegrees=10, twistLimitDegrees=0}
 {bone1Name=dyn_hairBk_c_06, bone2Name=dyn_hairBk_c_05, swing1LimitDegrees=20, swing2LimitDegrees=10, twistLimitDegrees=0}
 {
  bone1Name=dyn_hairBk_l_MSTR
  bone2Name=head
  swing1LimitDegrees=45
  swing1Motion=Free
  swing2LimitDegrees=45
  swing2Motion=Free
  twistLimitDegrees=45
  twistMotion=Free
 }
 {bone1Name=dyn_hairBk_l_01, bone2Name=dyn_hairBk_l_MSTR, swing1LimitDegrees=10, swing2LimitDegrees=10, twistLimitDegrees=0}
 {bone1Name=dyn_hairBk_l_02, bone2Name=dyn_hairBk_l_01, swing1LimitDegrees=10, swing2LimitDegrees=10, twistLimitDegrees=0}
 {bone1Name=dyn_hairBk_l_03, bone2Name=dyn_hairBk_l_02, swing1LimitDegrees=20, swing2LimitDegrees=20, twistLimitDegrees=0}
 {bone1Name=dyn_hairBk_l_04, bone2Name=dyn_hairBk_l_03, swing1LimitDegrees=20, swing2LimitDegrees=20, twistLimitDegrees=0}
 {bone1Name=dyn_hairBk_l_05, bone2Name=dyn_hairBk_l_04, swing1LimitDegrees=20, swing2LimitDegrees=20, twistLimitDegrees=0}
 {
  bone1Name=dyn_hairBk_r_MSTR
  bone2Name=head
  swing1LimitDegrees=45
  swing1Motion=Free
  swing2LimitDegrees=45
  swing2Motion=Free
  twistLimitDegrees=45
  twistMotion=Free
 }
 {bone1Name=dyn_hairBk_r_01, bone2Name=dyn_hairBk_r_MSTR, swing1LimitDegrees=45, swing2LimitDegrees=10, twistLimitDegrees=0}
 {bone1Name=dyn_hairBk_r_02, bone2Name=dyn_hairBk_r_01, swing1LimitDegrees=40, swing2LimitDegrees=10, twistLimitDegrees=0}
 {bone1Name=dyn_hairBk_r_03, bone2Name=dyn_hairBk_r_02, swing1LimitDegrees=40, swing2LimitDegrees=20, twistLimitDegrees=0}
 {bone1Name=dyn_hairBk_r_04, bone2Name=dyn_hairBk_r_03, swing1LimitDegrees=40, swing2LimitDegrees=20, twistLimitDegrees=0}
 {bone1Name=dyn_hairBk_r_05, bone2Name=dyn_hairBk_r_04, swing1LimitDegrees=40, swing2LimitDegrees=20, twistLimitDegrees=0}
 {
  bone1Name=dyn_hairMeshFront_l_MSTR
  bone2Name=head
  swing1LimitDegrees=45
  swing1Motion=Free
  swing2LimitDegrees=45
  swing2Motion=Free
  twistLimitDegrees=45
  twistMotion=Free
 }
 {bone1Name=dyn_hairMeshFront_l_01, bone2Name=dyn_hairMeshFront_l_MSTR, swing1LimitDegrees=10, swing2LimitDegrees=10, twistLimitDegrees=0}
 {bone1Name=dyn_hairMeshFront_l_02, bone2Name=dyn_hairMeshFront_l_01, swing1LimitDegrees=20, swing2LimitDegrees=20, twistLimitDegrees=0}
 {bone1Name=dyn_hairMeshFront_l_03, bone2Name=dyn_hairMeshFront_l_02, swing1LimitDegrees=20, swing2LimitDegrees=20, twistLimitDegrees=0}
 {bone1Name=dyn_hairMeshFront_l_04, bone2Name=dyn_hairMeshFront_l_03, swing1LimitDegrees=20, swing2LimitDegrees=20, twistLimitDegrees=0}
 {bone1Name=dyn_hairMeshFront_l_05, bone2Name=dyn_hairMeshFront_l_04, swing1LimitDegrees=20, swing2LimitDegrees=20, twistLimitDegrees=0}
 {bone1Name=dyn_hairMeshFront_l_06, bone2Name=dyn_hairMeshFront_l_05, swing1LimitDegrees=20, swing2LimitDegrees=20, twistLimitDegrees=0}
 {
  bone1Name=dyn_hairMesh_l_MSTR
  bone2Name=head
  swing1LimitDegrees=45
  swing1Motion=Free
  swing2LimitDegrees=45
  swing2Motion=Free
  twistLimitDegrees=45
  twistMotion=Free
 }
 {bone1Name=dyn_hairMesh_l_01, bone2Name=dyn_hairMesh_l_MSTR, swing1LimitDegrees=20, swing2LimitDegrees=20, twistLimitDegrees=0}
 {bone1Name=dyn_hairMesh_l_02, bone2Name=dyn_hairMesh_l_01, swing1LimitDegrees=20, swing2LimitDegrees=20, twistLimitDegrees=0}
 {bone1Name=dyn_hairMesh_l_03, bone2Name=dyn_hairMesh_l_02, swing1LimitDegrees=30, swing2LimitDegrees=30, twistLimitDegrees=0}
 {
  bone1Name=dyn_hairMesh_r_MSTR
  bone2Name=head
  swing1LimitDegrees=45
  swing1Motion=Free
  swing2LimitDegrees=45
  swing2Motion=Free
  twistLimitDegrees=45
  twistMotion=Free
 }
 {bone1Name=dyn_hairMesh_r_01, bone2Name=dyn_hairMesh_r_MSTR, swing1LimitDegrees=20, swing2LimitDegrees=20, twistLimitDegrees=0}
 {bone1Name=dyn_hairMesh_r_02, bone2Name=dyn_hairMesh_r_01, swing1LimitDegrees=20, swing2LimitDegrees=20, twistLimitDegrees=0}
 {bone1Name=dyn_hairMesh_r_03, bone2Name=dyn_hairMesh_r_02, swing1LimitDegrees=20, swing2LimitDegrees=20, twistLimitDegrees=0}
 {
  bone1Name=dyn_hairTop_c_MSTR
  bone2Name=head
  swing1LimitDegrees=45
  swing1Motion=Free
  swing2LimitDegrees=45
  swing2Motion=Free
  twistLimitDegrees=45
  twistMotion=Free
 }
 {
  bone1Name=dyn_hairTop_c_01
  bone2Name=dyn_hairTop_c_MSTR
  swing1LimitDegrees=10
  swing1Motion=Locked
  swing2LimitDegrees=10
  swing2Motion=Locked
  twistLimitDegrees=0
  twistMotion=Locked
 }
 {
  bone1Name=head
  bone2Name=neck_02
  swing1LimitDegrees=45
  swing1Motion=Free
  swing2LimitDegrees=45
  swing2Motion=Free
  twistLimitDegrees=45
  twistMotion=Free
 }
 {
  bone1Name=neck_02
  bone2Name=spine_05
  swing1LimitDegrees=45
  swing1Motion=Free
  swing2LimitDegrees=45
  swing2Motion=Free
  twistLimitDegrees=45
  twistMotion=Free
 }
 {
  bone1Name=clavicle_l
  bone2Name=spine_05
  swing1LimitDegrees=45
  swing1Motion=Free
  swing2LimitDegrees=45
  swing2Motion=Free
  twistLimitDegrees=45
  twistMotion=Free
 }
 {
  bone1Name=clavicle_r
  bone2Name=spine_05
  swing1LimitDegrees=45
  swing1Motion=Free
  swing2LimitDegrees=45
  swing2Motion=Free
  twistLimitDegrees=45
  twistMotion=Free
 }
]
