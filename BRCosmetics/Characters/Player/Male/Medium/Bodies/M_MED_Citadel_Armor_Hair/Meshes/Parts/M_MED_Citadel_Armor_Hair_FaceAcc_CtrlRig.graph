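# Control Rig for the Citadel Armor male medium character's hair, correcting the three back-hair master bones as the head moves.
# Two spherical pose readers track the head bone relative to spine_05, with opposite driver axes.
# Each reader's output weights transform offsets on the left, right and center bones: the first translates and rotates, the second only rotates.
# Two fixed-offset chains and a spine_05 rotation-constraint chain are unconnected to execution.
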ControlRigBlueprint /BRCosmetics/Characters/Player/Male/Medium/Bodies/M_MED_Citadel_Armor_Hair/Meshes/Parts/M_MED_Citadel_Armor_Hair_FaceAcc_CtrlRig
graphs [
 RigVMFunctionLibrary {links [], nodes []}
 RigVMModel {nodes [
   &RigUnit_ModifyTransforms(
    ExecutePin=()
    ItemToModify=((Item=(Type=Bone,Name="dyn_hairBk_l_MSTR"),Transform=(Rotation=(X=0,Y=0,Z=0,W=1),Translation=(X=-5,Y=0,Z=0),Scale3D=(X=1,Y=1,Z=1))))
    Weight=1
   )
   > &RigUnit_ModifyTransforms(
    ItemToModify=((Item=(Type=Bone,Name="dyn_hairBk_r_MSTR"),Transform=(Rotation=(X=0,Y=0,Z=0,W=1),Translation=(X=-5,Y=0,Z=0),Scale3D=(X=1,Y=1,Z=1))))
    Weight=1
   )
   > &RigUnit_ModifyTransforms(
    ItemToModify=((Item=(Type=Bone,Name="dyn_hairBk_c_MSTR"),Transform=(Rotation=(X=0,Y=0,Z=0,W=1),Translation=(X=-8,Y=0,Z=0),Scale3D=(X=1,Y=1,Z=1))))
    Weight=1
   )
   &RigUnit_ModifyTransforms(
    ExecutePin=()
    ItemToModify=((Item=(Type=Bone,Name="dyn_hairBk_l_MSTR"),Transform=(Rotation=(X=0,Y=0,Z=0,W=1),Translation=(X=0,Y=-3,Z=-5),Scale3D=(X=1,Y=1,Z=1))))
    Weight=1
   )
   > &RigUnit_ModifyTransforms(
    ItemToModify=((Item=(Type=Bone,Name="dyn_hairBk_r_MSTR"),Transform=(Rotation=(X=0,Y=0,Z=0,W=1),Translation=(X=-1.5,Y=-5,Z=5),Scale3D=(X=1,Y=1,Z=1))))
    Weight=1
   )
   > &RigUnit_ModifyTransforms(
    ItemToModify=((Item=(Type=Bone,Name="dyn_hairBk_c_MSTR"),Transform=(Rotation=(X=0,Y=0,Z=0,W=1),Translation=(X=-1.5,Y=-6,Z=0),Scale3D=(X=1,Y=1,Z=1))))
    Weight=1
   )
   RigUnit_BeginExecution = &RigUnit_BeginExecution()
   n1 = &RigUnit_SphericalPoseReader(
    ActiveRegionScaleFactors=(PositiveWidth=0.25,NegativeWidth=3,PositiveHeight=1,NegativeHeight=1)
    ActiveRegionSize=0.1
    Debug=(bDrawDebug=True,DebugScale=25,DebugSegments=20,DebugThickness=0.25)
    DriverAxis=(X=0,Y=-1,Z=0)
    DriverItem=(Type=Bone,Name="head")
    FalloffRegionScaleFactors=(PositiveWidth=1.2,NegativeWidth=1,PositiveHeight=2,NegativeHeight=2)
    FalloffSize=0.2
    OptionalParentItem=(Type=Bone,Name="spine_05")
    RotationOffset=(X=-80,Y=0,Z=-20)
    ExecutePin=@RigUnit_BeginExecution.ExecutePin
   )
   > &RigUnit_ModifyTransforms(
    ItemToModify=((Item=(Type=Bone,Name="dyn_hairBk_l_MSTR"),Transform=(Rotation=(X=0,Y=0,Z=0,W=1),Translation=(X=-3,Y=0,Z=0),Scale3D=(X=1,Y=1,Z=1))))
    ItemToModify.0.Transform.Rotation=&RigVMFunction_MathQuaternionFromEuler(Euler=(X=0,Y=20,Z=0)).Result
    Weight=@n1.OutputParam
   )
   > &RigUnit_ModifyTransforms(
    ItemToModify=((Item=(Type=Bone,Name="dyn_hairBk_r_MSTR"),Transform=(Rotation=(X=0,Y=0,Z=0,W=1),Translation=(X=-3,Y=0,Z=0),Scale3D=(X=1,Y=1,Z=1))))
    ItemToModify.0.Transform.Rotation=&RigVMFunction_MathQuaternionFromEuler(Euler=(X=10,Y=-10,Z=-10)).Result
    Weight=@n1.OutputParam
   )
   > &RigUnit_ModifyTransforms(
    ItemToModify=((Item=(Type=Bone,Name="dyn_hairBk_c_MSTR"),Transform=(Rotation=(X=0,Y=0,Z=0,W=1),Translation=(X=-3,Y=0,Z=0),Scale3D=(X=1,Y=1,Z=1))))
    ItemToModify.0.Transform.Rotation=&RigVMFunction_MathQuaternionFromEuler(Euler=(X=0,Y=0,Z=0)).Result
    Weight=@n1.OutputParam
   )
   > n2 = &RigUnit_SphericalPoseReader(
    ActiveRegionScaleFactors=(PositiveWidth=0.25,NegativeWidth=0.25,PositiveHeight=0.25,NegativeHeight=0.25)
    ActiveRegionSize=0.1
    Debug=(bDrawDebug=false,DebugScale=25,DebugSegments=20,DebugThickness=0.25)
    DriverAxis=(X=0,Y=1,Z=0)
    DriverItem=(Type=Bone,Name="head")
    FalloffRegionScaleFactors=(PositiveWidth=0.75,NegativeWidth=0.5,PositiveHeight=4,NegativeHeight=4)
    FalloffSize=0.2
    OptionalParentItem=(Type=Bone,Name="spine_05")
    RotationOffset=(X=90,Y=0,Z=25)
   )
   > &RigUnit_ModifyTransforms(
    ItemToModify=((Item=(Type=Bone,Name="dyn_hairBk_l_MSTR"),Transform=(Rotation=(X=0,Y=0,Z=0,W=1),Translation=(X=0,Y=0,Z=0),Scale3D=(X=1,Y=1,Z=1))))
    ItemToModify.0.Transform.Rotation=&RigVMFunction_MathQuaternionFromEuler(Euler=(X=0,Y=30,Z=-20)).Result
    Weight=@n2.OutputParam
   )
   > &RigUnit_ModifyTransforms(
    ItemToModify=((Item=(Type=Bone,Name="dyn_hairBk_r_MSTR"),Transform=(Rotation=(X=0,Y=0,Z=0,W=1),Translation=(X=0,Y=0,Z=0),Scale3D=(X=1,Y=1,Z=1))))
    ItemToModify.0.Transform.Rotation=&RigVMFunction_MathQuaternionFromEuler(Euler=(X=0,Y=-30,Z=-30)).Result
    Weight=@n2.OutputParam
   )
   > &RigUnit_ModifyTransforms(
    ItemToModify=((Item=(Type=Bone,Name="dyn_hairBk_c_MSTR"),Transform=(Rotation=(X=0,Y=0,Z=0,W=1),Translation=(X=0,Y=0,Z=0),Scale3D=(X=1,Y=1,Z=1))))
    ItemToModify.0.Transform.Rotation=&RigVMFunction_MathQuaternionFromEuler(Euler=(X=0,Y=0,Z=-20)).Result
    Weight=@n2.OutputParam
   )
   &RigUnit_RotationConstraintLocalSpaceOffset(
    AdvancedSettings=(RotationOrderForFilter=XZY)
    Child=(Type=Bone,Name="dyn_hairBk_l_MSTR")
    ExecutePin=()
    Filter=(bX=True,bY=True,bZ=True)
    Parents=((Item=(Type=Bone,Name="spine_05"),Weight=1))
    Weight=1
   )
   > &RigUnit_RotationConstraintLocalSpaceOffset(
    AdvancedSettings=(RotationOrderForFilter=XZY)
    Child=(Type=Bone,Name="dyn_hairBk_r_MSTR")
    Filter=(bX=True,bY=True,bZ=True)
    Parents=((Item=(Type=Bone,Name="spine_05"),Weight=1))
    Weight=1
   )
   > &RigUnit_RotationConstraintLocalSpaceOffset(
    AdvancedSettings=(RotationOrderForFilter=XZY)
    Child=(Type=Bone,Name="dyn_hairBk_c_MSTR")
    Filter=(bX=True,bY=True,bZ=True)
    Parents=((Item=(Type=Bone,Name="spine_05"),Weight=1))
    Weight=1
   )
  ]}
]
hierarchy [
 root {loc=<0 0 0>, rot=<0 0 0 1>}
 pelvis {parent=root, loc=<0 -0.7071 93.1191>, rot=<0.0127 0.707 -0.0127 -0.707>}
 spine_01 {parent=pelvis, loc=<6.7494 -0.6388 0>, rot=<0 0 0.034 -0.9994>}
 spine_02 {parent=spine_01, loc=<7.2518 -0.035 0>, rot=<0 0 0.0238 -0.9997>}
 spine_03 {parent=spine_02, loc=<6.0104 0 0>, rot=<0 0 0.0149 -0.9999>}
 spine_04 {parent=spine_03, loc=<8.1024 0 0>, rot=<0 0 -0.0132 -0.9999>}
 spine_05 {parent=spine_04, loc=<10.0146 0 0>, rot=<0 0 -0.0015 -1>}
 neck_01 {parent=spine_05, loc=<11.3138 0.1529 0>, rot=<0 0 -0.2003 -0.9797>}
 neck_02 {parent=neck_01, loc=<4.5627 0 0>, rot=<0 0 0.0504 -0.9987>}
 head {parent=neck_02, loc=<5.3914 0 0>, rot=<0 0 0.115 -0.9934>}
 dyn_hairTop_c_MSTR {parent=head, loc=<15.0502 1.085 0.9183>, rot=<0.0041 -1 0 0>}
 dyn_hairTop_c_01 {parent=dyn_hairTop_c_MSTR, loc=<0 0 0>, rot=<0 0 0 1>}
 dyn_hairBk_l_MSTR {parent=head, loc=<9.1503 -4.7549 -6.032>, rot=<0.0034 -0.8262 -0.5633 0.0023>}
 dyn_hairBk_l_01 {parent=dyn_hairBk_l_MSTR, loc=<0 0 0>, rot=<0 0 0 1>}
 dyn_hairBk_l_02 {parent=dyn_hairBk_l_01, loc=<4.5656 0 0>, rot=<0.0009 -0.0275 0.0376 -0.9989>}
 dyn_hairBk_l_03 {parent=dyn_hairBk_l_02, loc=<4.5656 0 0>, rot=<-0.017 -0.0795 0.2035 -0.9757>}
 dyn_hairBk_l_04 {parent=dyn_hairBk_l_03, loc=<4.5656 0 0>, rot=<-0.0006 -0.0259 0.0259 -0.9993>}
 dyn_hairBk_l_05 {parent=dyn_hairBk_l_04, loc=<4.5656 0 0>, rot=<-0.0006 -0.0259 0.0259 -0.9993>}
 dyn_hairBk_c_MSTR {parent=head, loc=<10.505 -7.1401 -0.3531>, rot=<0.0041 -1 0 0>}
 dyn_hairBk_c_01 {parent=dyn_hairBk_c_MSTR, loc=<0 0 0>, rot=<0 0 0 1>}
 dyn_hairBk_c_02 {parent=dyn_hairBk_c_01, loc=<4.9571 0 0>, rot=<0 0 0.018 -0.9998>}
 dyn_hairBk_c_03 {parent=dyn_hairBk_c_02, loc=<4.9571 0 0>, rot=<0 0 0.1866 -0.9824>}
 dyn_hairBk_c_04 {parent=dyn_hairBk_c_03, loc=<4.9571 0 0>, rot=<0 0 0.1321 -0.9912>}
 dyn_hairBk_c_05 {parent=dyn_hairBk_c_04, loc=<4.9571 0 0>, rot=<0 0 -0.0669 -0.9978>}
 dyn_hairBk_c_06 {parent=dyn_hairBk_c_05, loc=<4.9571 0 0>, rot=<0 0 -0.0669 -0.9978>}
 dyn_hairBk_r_MSTR {parent=head, loc=<8.8888 -5.1697 5.1689>, rot=<0.0039 -0.9466 0.3223 -0.0013>}
 dyn_hairBk_r_01 {parent=dyn_hairBk_r_MSTR, loc=<0 0 0>, rot=<0 0 0 1>}
 dyn_hairBk_r_02 {parent=dyn_hairBk_r_01, loc=<4.4458 0 0>, rot=<0.0595 0.0131 0.0183 -0.998>}
 dyn_hairBk_r_03 {parent=dyn_hairBk_r_02, loc=<4.4458 0 0>, rot=<0.0338 0.0066 0.2301 -0.9726>}
 dyn_hairBk_r_04 {parent=dyn_hairBk_r_03, loc=<4.4458 0 0>, rot=<-0.0054 0.0148 0.0907 -0.9958>}
 dyn_hairBk_r_05 {parent=dyn_hairBk_r_04, loc=<4.4458 0 0>, rot=<-0.0054 0.0148 0.0907 -0.9958>}
 dyn_hairMesh_r_MSTR {parent=head, loc=<2.1807 -0.4294 5.9223>, rot=<0.0546 0.6453 0.7526 0.1193>}
 dyn_hairMesh_r_01 {parent=dyn_hairMesh_r_MSTR, loc=<0 0 0>, rot=<0 0 0 1>}
 dyn_hairMesh_r_02 {parent=dyn_hairMesh_r_01, loc=<3.9295 0.1628 0.3631>, rot=<0.0407 0.195 -0.0601 -0.9781>}
 dyn_hairMesh_r_03 {parent=dyn_hairMesh_r_02, loc=<5.1398 0.0408 0.1336>, rot=<-0.0109 0.1353 0.0114 -0.9907>}
 dyn_hairMesh_l_MSTR {parent=head, loc=<2.0264 -0.6354 -6.3617>, rot=<-0.0258 -0.6431 0.7577 0.1079>}
 dyn_hairMesh_l_01 {parent=dyn_hairMesh_l_MSTR, loc=<0 0 0>, rot=<0 0 0 1>}
 dyn_hairMesh_l_02 {parent=dyn_hairMesh_l_01, loc=<3.1606 -0.2411 -0.6106>, rot=<-0.0555 -0.1509 -0.0293 -0.9866>}
 dyn_hairMesh_l_03 {parent=dyn_hairMesh_l_02, loc=<4.8871 0.0832 -0.2355>, rot=<-0.0202 -0.1477 0.0013 -0.9888>}
 dyn_hairMeshFront_l_MSTR {parent=head, loc=<9.9217 9.4475 -1.8688>, rot=<-0.3799 -0.4323 -0.2528 -0.7778>}
 dyn_hairMeshFront_l_01 {parent=dyn_hairMeshFront_l_MSTR, loc=<0 0 0>, rot=<0 0 0 1>}
 dyn_hairMeshFront_l_02 {parent=dyn_hairMeshFront_l_01, loc=<2.1301 0 0>, rot=<0.0544 -0.7134 0.1279 -0.6868>}
 dyn_hairMeshFront_l_03 {parent=dyn_hairMeshFront_l_02, loc=<4.6674 -0.0228 -0.5571>, rot=<0.1793 -0.2269 0.0517 -0.9559>}
 dyn_hairMeshFront_l_04 {parent=dyn_hairMeshFront_l_03, loc=<5.0955 0 0>, rot=<-0.0161 -0.1436 0.0394 -0.9887>}
 dyn_hairMeshFront_l_05 {parent=dyn_hairMeshFront_l_04, loc=<5.0982 0 0>, rot=<-0.0353 -0.1926 0.0313 -0.9801>}
 dyn_hairMeshFront_l_06 {parent=dyn_hairMeshFront_l_05, loc=<3.7985 0 0>, rot=<-0.0087 -0.2247 0.0544 -0.9729>}
 clavicle_l {parent=spine_05, loc=<4.3575 -0.1466 -2.3223>, rot=<0.0465 -0.7289 -0.009 -0.6829>}
 upperarm_l {parent=clavicle_l, loc=<15.5111 0 0>, rot=<0.0226 -0.3234 -0.0088 -0.9459>}
 upperarm_twist_01_l {parent=upperarm_l, loc=<0.0041 0 0>, rot=<0 0 0 -1>}
 clavicle_r {parent=spine_05, loc=<4.3575 -0.1466 2.3223>, rot=<0.7289 0.0465 -0.6829 0.009>}
 upperarm_r {parent=clavicle_r, loc=<-15.5111 0 0>, rot=<0.0226 -0.3234 -0.0088 -0.9459>}
]
