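# Anim Blueprint for the Citadel Armor Hair male character's FaceAcc part, driving hair secondary motion.
# Parent-mesh pose feeds two Control Rigs, spine/head rotation constraints, then head-space RigidBody sims switched for frontend, gameplay/emote and skydive.
# Additive bone offsets reshape back hair for the custom cape, skydive floating and backbling-free jumps.
# EventGraph derives gravity, backbling-based acceleration scale, cape tag checks and a melee head-rotation alpha for the first rig.
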AnimBlueprint /BRCosmetics/Characters/Player/Male/Medium/Bodies/M_MED_Citadel_Armor_Hair/Meshes/Parts/M_MED_Citadel_Armor_Hair_FaceAcc_AnimBP
target_skeleton=&/M_MED_Citadel_Armor_Hair_FaceAcc_Skeleton
graphs [
 AnimGraph {nodes [
   n1 = Control Rig(Source=Copy Pose From Mesh() {bUseAttachedParent=true}, Alpha=Get MeleeHeadDwn() -> (MeleeHeadDwn:double)) {
    alphaBoolBlend {blendOption=Linear}
    controlRigAssetReference {blueprintRigClass=&/M_MED_Citadel_Armor_Hair_FaceAccConstraint_CtrlRig.M_MED_Citadel_Armor_Hair_FaceAccConstraint_CtrlRig_C}
   }
   n2 = Control Rig(Source=@n1) {
    alphaBoolBlend {blendOption=Linear}
    controlRigAssetReference {blueprintRigClass=&/M_MED_Citadel_Armor_Hair_FaceAcc_CtrlRig.M_MED_Citadel_Armor_Hair_FaceAcc_CtrlRig_C}
   }
   n3 = "Constraint - dyn_hairMesh_r_MSTR (RR)"!(ComponentPose=Local To Component(LocalPose=@n2), Alpha=1, ConstraintWeights_0=0.5, ConstraintWeights_1:float=0.5) -> (Pose) {
    alphaBoolBlend {blendOption=Linear}
    boneToModify {boneName=dyn_hairMesh_r_MSTR}
    constraintSetup [{transformType=Rotation, targetBone {boneName=spine_05}}, {transformType=Rotation, targetBone {boneName=head}}]
    constraintWeights [1,1]
   }
   n4 = "Constraint - dyn_hairMesh_l_MSTR (RR)"!(ComponentPose=@n3, Alpha=1, ConstraintWeights_0=0.5, ConstraintWeights_1:float=0.5) -> (Pose) {
    alphaBoolBlend {blendOption=Linear}
    boneToModify {boneName=dyn_hairMesh_l_MSTR}
    constraintSetup [{transformType=Rotation, targetBone {boneName=spine_05}}, {transformType=Rotation, targetBone {boneName=head}}]
    constraintWeights [1,1]
   }
   n5 = RigidBody!(
    ComponentPose=@n4
    bAlphaBoolEnabled:bool=Get inFrontNoEmote() -> (inFrontNoEmote:bool)
    OverrideWorldGravity:Vector="0,0,-980"
    ExternalForce
    ComponentLinearAccScale:Vector="0.4,0.3,0.3"
    ComponentLinearVelScale:Vector="0.4,0.3,0.3"
   ) -> (Pose) {
    alphaInputType=Bool
    bClampLinearTranslationLimitToRefPose=true
    bOverrideWorldGravity=true
    overridePhysicsAsset=&/M_MED_Citadel_Armor_FaceAcc_PhysicsFrontend
    simulationSpace=BaseBoneSpace
    alphaBoolBlend {blendOption=Linear}
    baseBoneRef {boneName=head}
    componentAppliedLinearAccClamp=<3000 3000 3000>
    componentLinearAccScale=<0.4 0.3 0.3>
    componentLinearVelScale=<0.4 0.3 0.3>
    overrideWorldGravity {z=-980}
    simSpaceSettings {
     maxAngularAcceleration=1000
     maxAngularVelocity=3
     maxLinearAcceleration=250
     maxLinearVelocity=500
     velocityScaleZ=0
     worldAlpha=0.25
     externalLinearDragV {y=0.5}
    }
   }
   n6 = Get GravityOverride()
   n7 = Get AccScaleBackBling() -> (AccScaleBackBling:Vector)
   n8 = RigidBody(
    ComponentPose=@n5
    bAlphaBoolEnabled=Get NoFrontendorEmoting() -> (NoFrontendorEmoting:bool)
    OverrideWorldGravity=@n6
    ComponentLinearAccScale=@n7
    ComponentLinearVelScale=@n7
   ) {
    alphaInputType=Bool
    bClampLinearTranslationLimitToRefPose=true
    bOverrideWorldGravity=true
    overridePhysicsAsset=&/M_MED_Citadel_Armor_FaceAcc_Physics1
    simulationSpace=BaseBoneSpace
    alphaBoolBlend {blendOption=Linear}
    baseBoneRef {boneName=head}
    componentAppliedLinearAccClamp=<3000 3000 3000>
    componentLinearAccScale=<0.4 0.3 0.3>
    componentLinearVelScale=<0.4 0.3 0.3>
    overrideWorldGravity {z=-980}
    simSpaceSettings {
     maxAngularAcceleration=1000
     maxAngularVelocity=3
     maxLinearAcceleration=250
     maxLinearVelocity=500
     velocityScaleZ=0
     worldAlpha=0.25
     externalLinearDragV {y=0.5}
    }
   }
   n9 = RigidBody!(ComponentPose=@n8, bAlphaBoolEnabled=Get bIsSkyDiving(), OverrideWorldGravity=@n6, ExternalForce) -> (Pose) {
    alphaInputType=Bool
    bClampLinearTranslationLimitToRefPose=true
    bOverrideWorldGravity=true
    overridePhysicsAsset=&/M_MED_Citadel_Armor_FaceAcc_PhysicsSkydive
    simulationSpace=BaseBoneSpace
    alphaBoolBlend {blendOption=Linear}
    baseBoneRef {boneName=head}
    componentAppliedLinearAccClamp=<3000 3000 3000>
    componentLinearAccScale=<0.4 0.3 0.3>
    componentLinearVelScale=<0.4 0.3 0.3>
    overrideWorldGravity {z=-980}
    simSpaceSettings {
     maxAngularAcceleration=1000
     maxAngularVelocity=3
     maxLinearAcceleration=250
     maxLinearVelocity=500
     velocityScaleZ=0
     worldAlpha=0.25
     externalLinearDragV {y=0.5}
    }
   }
   n10 = Get isCustomCape() -> (isCustomCape:bool)
   n11 = "Transform (Modify) Bone - Bone: dyn_hairBk_l_03"!(ComponentPose=@n9, bAlphaBoolEnabled:bool=@n10, Translation="0,0,0", Rotation="0,0,20", Scale="0.95,0.95,0.95") -> (Pose) {
    alphaInputType=Bool
    rotationMode=BMM_Additive
    scaleMode=BMM_Additive
    translationMode=BMM_Additive
    alphaBoolBlend {blendOption=Linear}
    alphaScaleBiasClamp {bInterpResult=true}
    boneToModify {boneName=dyn_hairBk_l_03}
    rotation {roll=20}
    scale=<0.95 0.95 0.95>
   }
   n12 = "Transform (Modify) Bone - Bone: dyn_hairBk_c_03"!(ComponentPose=@n11, bAlphaBoolEnabled:bool=@n10, Translation="0,0,1", Rotation="0,0,5", Scale="0.95,0.95,0.95") -> (Pose) {
    alphaInputType=Bool
    rotationMode=BMM_Additive
    scaleMode=BMM_Additive
    translationMode=BMM_Additive
    alphaBoolBlend {blendOption=Linear}
    alphaScaleBiasClamp {bInterpResult=true}
    boneToModify {boneName=dyn_hairBk_c_03}
    rotation {roll=5}
    scale=<0.95 0.95 0.95>
    translation {z=1}
   }
   n13 = "Transform (Modify) Bone - Bone: dyn_hairBk_r_03"!(ComponentPose=@n12, bAlphaBoolEnabled:bool=@n10, Translation="0,0,0", Rotation="0,0,20", Scale="0.95,0.95,0.95") -> (Pose) {
    alphaInputType=Bool
    rotationMode=BMM_Additive
    scaleMode=BMM_Additive
    translationMode=BMM_Additive
    alphaBoolBlend {blendOption=Linear}
    alphaScaleBiasClamp {bInterpResult=true}
    boneToModify {boneName=dyn_hairBk_r_03}
    rotation {roll=20}
    scale=<0.95 0.95 0.95>
   }
   n14 = Get bIsSkydiveFloating()
   n15 = "Transform (Modify) Bone - Bone: dyn_hairBk_l_03"(ComponentPose=@n13, bAlphaBoolEnabled=@n14, Translation="0,0,0", Rotation="0,0,20", Scale="0.9,0.9,0.9") {
    alphaInputType=Bool
    rotationMode=BMM_Additive
    scaleMode=BMM_Additive
    translationMode=BMM_Additive
    alphaBoolBlend {blendOption=Linear}
    alphaScaleBiasClamp {bInterpResult=true}
    boneToModify {boneName=dyn_hairBk_l_03}
    rotation {roll=20}
    scale=<0.9 0.9 0.9>
   }
   n16 = "Transform (Modify) Bone - Bone: dyn_hairBk_c_03"(ComponentPose=@n15, bAlphaBoolEnabled=@n14, Translation="0,0,3", Rotation="0,0,20", Scale="0.9,0.9,0.9") {
    alphaInputType=Bool
    rotationMode=BMM_Additive
    scaleMode=BMM_Additive
    translationMode=BMM_Additive
    alphaBoolBlend {blendOption=Linear}
    alphaScaleBiasClamp {bInterpResult=true}
    boneToModify {boneName=dyn_hairBk_c_03}
    rotation {roll=20}
    scale=<0.9 0.9 0.9>
    translation {z=3}
   }
   n17 = "Transform (Modify) Bone - Bone: dyn_hairBk_r_03"(ComponentPose=@n16, bAlphaBoolEnabled=@n14, Translation="0,0,0", Rotation="0,0,20", Scale="0.9,0.9,0.9") {
    alphaInputType=Bool
    rotationMode=BMM_Additive
    scaleMode=BMM_Additive
    translationMode=BMM_Additive
    alphaBoolBlend {blendOption=Linear}
    alphaScaleBiasClamp {bInterpResult=true}
    boneToModify {boneName=dyn_hairBk_r_03}
    rotation {roll=20}
    scale=<0.9 0.9 0.9>
   }
   n18 = Get Is Jumping() -> (Is Jumping:double)
   n19 = "Transform (Modify) Bone - Bone: dyn_hairBk_c_04"(ComponentPose=@n17, Alpha=@n18, Translation="0,0,0", Rotation="0,0,70", Scale="0.9,0.9,0.9") {
    rotationMode=BMM_Additive
    scaleMode=BMM_Additive
    translationMode=BMM_Additive
    alphaBoolBlend {blendOption=Linear}
    alphaScaleBiasClamp {bInterpResult=true, interpSpeedDecreasing=1, interpSpeedIncreasing=1}
    boneToModify {boneName=dyn_hairBk_c_04}
    rotation {roll=70}
    scale=<0.9 0.9 0.9>
   }
   n20 = "Transform (Modify) Bone - Bone: dyn_hairBk_l_03"!(ComponentPose=@n19, Alpha:float=@n18, Translation="0,0,0", Rotation="0,0,40", Scale="0.9,0.9,0.9") -> (Pose) {
    rotationMode=BMM_Additive
    scaleMode=BMM_Additive
    translationMode=BMM_Additive
    alphaBoolBlend {blendOption=Linear}
    alphaScaleBiasClamp {bInterpResult=true, interpSpeedDecreasing=1, interpSpeedIncreasing=1}
    boneToModify {boneName=dyn_hairBk_l_03}
    rotation {roll=40}
    scale=<0.9 0.9 0.9>
   }
   n21 = "Transform (Modify) Bone - Bone: dyn_hairBk_r_03"!(ComponentPose=@n20, Alpha:float=@n18, Translation="0,0,0", Rotation="0,0,30", Scale="0.9,0.9,0.9") -> (Pose) {
    rotationMode=BMM_Additive
    scaleMode=BMM_Additive
    translationMode=BMM_Additive
    alphaBoolBlend {blendOption=Linear}
    alphaScaleBiasClamp {bInterpResult=true, interpSpeedDecreasing=1, interpSpeedIncreasing=1}
    boneToModify {boneName=dyn_hairBk_r_03}
    rotation {roll=30}
    scale=<0.9 0.9 0.9>
   }
   Output Pose(Result=Component To Local(ComponentPose=@n21)) {layerGroup=None}
   NOT Boolean(A=Get bIsPlayingMeleeAnim())
  ]}
 EventGraph {nodes [
   Event BlueprintUpdateAnimation()
   > n1 = Sequence()
   n2 = CheckForBackbling()
   > Set isWearingBackbling(isWearingBackbling=OR Boolean(A=@n2, B=@n2.bIsBackblingLong))
   > Set isWearingCapeBool(isWearingCapeBool:bool=CheckForBackbling().bIsBackblingACape) -> (Output_Get:bool)
   n3 = Sequence!(execute=@n1.then_1) -> (then_0, then_1, then_2:Exec, then_3:Exec, then_4:Exec, then_5:Exec)
   > n4 = Gravity Override State Function DEPRECATED(CharacterPartAnimInstance=Self-Reference(), GameTimeInSeconds=GetGameTimeInSeconds(), GravityOverrideParams=Get Gravity Override Params())
   > Set GravityOverride(GravityOverride=@n4)
   > n5 = Branch(Condition=Get isWearingBackbling())
   > Set Is Jumping(Is Jumping:double=0) -> (Output_Get:double)
   Set Is Jumping(execute=@n5.else, Is Jumping="To Float (Boolean)"(InBool=Get bIsJumping()))
   n6 = Branch(
    execute=@n3.then_1
    Condition=BackpackItemDefinitionHasAnyMetaTag!(Tags:Array of Gameplay Tag Structures=Make Array("[0]"=(TagName="Cosmetics.MetaTags.Citadel.Armor.Cape")), bRequiresEquipped:bool=false) -> (ReturnValue:bool)
   )
   > Set isCustomCape(isCustomCape:bool=true) -> (Output_Get:bool)
   Set isCustomCape(execute=@n6.else, isCustomCape=false)
   Set isWearingCape(execute=@n3.then_2, isWearingCape=OR Boolean(A=Get isCustomCape(), B=Get isWearingCapeBool() -> (isWearingCapeBool:bool)))
   n7 = Get isCustomCape()
   n8 = Branch(execute=@n3.then_3, Condition=AND Boolean(A=Get isWearingBackbling(), B=NOT Boolean(A=@n7)))
   > Set AccScaleBackBling(AccScaleBackBling:Vector="0.1,0,0.1") -> (Output_Get:Vector)
   Set AccScaleBackBling(execute=@n8.else, AccScaleBackBling="0.4,0.3,0.3")
   n9 = Get bIsPlayingEmote()
   Set inFrontNoEmote(execute=@n3.then_4, inFrontNoEmote:bool=AND Boolean(A=Get bIsInFrontEnd(), B=NOT Boolean(A=@n9))) -> (Output_Get:bool)
   n10 = Get bIsInFrontEnd()
   > Set NoFrontendorEmoting(NoFrontendorEmoting:bool=OR Boolean(A=Get bIsPlayingEmote(), B=NOT Boolean(A=@n10))) -> (Output_Get:bool)
   n11 = Set headRotBack(
    execute=@n3.then_5
    headRotBack:double=GetAlphaFromReferenceTransformDelta(SkeletalMeshComponent=GetOwnerSkeletalMeshComponent(), BoneName=head, BaseName=spine_05, InputRange=(X=20,Y=-20), TwistAxis="1,0,0")
   ) -> (Output_Get:double)
   > n12 = Branch(Condition=Get bIsPlayingMeleeAnim())
   > Set MeleeHeadDwn(MeleeHeadDwn:double=@n11) -> (Output_Get:double)
   Set MeleeHeadDwn(execute=@n12.else, MeleeHeadDwn=1)
   GetCurveValue(CurveName=SPR_arm_front_alpha)
   Event OnCharacterCustomizationLoaded()
  ]}
]
variables [
 AccScaleBackBling {type=Vector}
 Gravity Override Params {type=GravityOverrideParamsStruct}
 GravityOverride {type=Vector}
 headRotBack {type=number}
 inFrontNoEmote {type=boolean}
 Is Jumping {type=number}
 isCustomCape {type=boolean}
 isWearingBackbling {type=boolean}
 isWearingCape {type=boolean}
 isWearingCapeBool {type=boolean}
 MeleeHeadDwn {type=number}
 NoFrontendorEmoting {type=boolean}
]
